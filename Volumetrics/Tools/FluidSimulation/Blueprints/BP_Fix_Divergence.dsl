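; Fluid-simulation tool actor that removes divergence from a seeded 2D velocity field using material draws into render targets.
; Construction creates MIDs and previews; BeginPlay creates velocity, pressure, divergence and scratch targets, seeds them, then enables tick after a short delay.
; Each tick ping-pongs Jacobi pressure iterations, subtracts the gradient, copies velocity back and to RT_Default, then recomputes divergence.
; An Advect function with force injection exists but is never called here.
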(event Custom|Create%20RTs
  (bind _2d_resolution (Variables|Default|Get2D%20Resolution))
  (bind _returnvalue (Rendering|CreateRenderTarget2D _2d_resolution _2d_resolution))
  (Variables|Default|SetVelocity%20Texture _returnvalue)
  (bind _returnvalue_2 (Rendering|CreateRenderTarget2D _2d_resolution _2d_resolution))
  (Variables|Default|SetPressure%20Texture _returnvalue_2)
  (bind _returnvalue_3 (Rendering|CreateRenderTarget2D (Variables|Default|Get2D%20Resolution) (Variables|Default|Get2D%20Resolution)))
  (Variables|Default|SetDivergence%20Texture _returnvalue_3)
  (bind _returnvalue_4 (Rendering|CreateRenderTarget2D (Variables|Default|Get2D%20Resolution) (Variables|Default|Get2D%20Resolution)))
  (Variables|Default|SetScratch%20RT _returnvalue_4)
  (Rendering|Material|SetTextureParameterValue (Variables|MIDs|GetVelocity%20Preview%20MID) "RT" (Variables|Default|GetVelocity%20Texture))
  (Rendering|Material|SetTextureParameterValue (Variables|MIDs|GetPressure%20Preview%20MID) "RT" (Variables|Default|GetPressure%20Texture))
  (Rendering|Material|SetTextureParameterValue (Variables|MIDs|GetDivergence%20Preview%20MID) "RT" (Variables|Default|GetDivergence%20Texture))
  (Rendering|Material|SetTextureParameterValue (Variables|MIDs|GetDivergence%20MID) "Velocity" (Variables|Default|GetVelocity%20Texture)))

(event Custom|Seed%20Preview
  (Rendering|ClearRenderTarget2D "/Volumetrics/Content/VolumeTextures/Textures/RT_Default.RT_Default")
  (Rendering|DrawMaterialtoRenderTarget "/Volumetrics/Content/VolumeTextures/Textures/RT_Default.RT_Default" (Variables|MIDs|GetVelocity%20Seed%20MID)))

(event Custom|Seed%20Textures
  (Rendering|DrawMaterialtoRenderTarget (Variables|Default|GetVelocity%20Texture) (Variables|MIDs|GetVelocity%20Seed%20MID))
  (Rendering|DrawMaterialtoRenderTarget (Variables|Default|GetDivergence%20Texture) (Variables|MIDs|GetDivergence%20MID)))

(event EventBeginPlay
  (CallFunction|CreateRTs)
  (CallFunction|SetMIDParameters)
  (CallFunction|SeedTextures)
  (Utilities|FlowControl|Delay 0.2)
  (Actor|Tick|SetActorTickEnabled true))

(event EventTick
  (params (DeltaSeconds :type Float%20%28single-precision%29))
  (CallFunction|RemoveDivergence)
  (Math|Integer|IncrementInt (Variables|Default|GetCurrentPass)))

(event Custom|Remove%20Divergence
  (bind _velocity_texture (Variables|Default|GetVelocity%20Texture))
  ;[1 1 1] Simulation
  (CallFunction|PressureIteration (Variables|Default|GetPressure%20Texture) (Variables|Default|GetScratch%20RT) (Variables|Default|GetDivergence%20Texture))
  (CallFunction|GradientSubtraction (Variables|Default|GetVelocity%20Texture) (Variables|Default|GetPressure%20Texture) (Variables|Default|GetScratch%20RT))
  (CallFunction|CopyRenderTarget (Variables|Default|GetScratch%20RT) _velocity_texture)
  (CallFunction|CopyRenderTarget _velocity_texture "/Volumetrics/Content/VolumeTextures/Textures/RT_Default.RT_Default")
  (Rendering|DrawMaterialtoRenderTarget (Variables|Default|GetDivergence%20Texture) (Variables|MIDs|GetDivergence%20MID)))

(fn Advect
  (params (Source%20Velocity :type Texture%20Render%20Target%202D%20Object%20Reference) (Source%20Texture :type Texture%20Render%20Target%202D%20Object%20Reference) (Destination%20Texture :type Texture%20Render%20Target%202D%20Object%20Reference) (Is%20Velocity :type Boolean))
  (bind _advect_mid (Variables|MIDs|GetAdvect%20MID))
  (bind _advect_mid_2 (Variables|MIDs|GetAdvect%20MID))
  (Rendering|Material|SetTextureParameterValue _advect_mid_2 "Velocity" Source%20Velocity)
  (Rendering|Material|SetTextureParameterValue _advect_mid_2 "Source" Source%20Texture)
  (Rendering|Material|SetScalarParameterValue _advect_mid_2 "IsVelocity" (Math|Conversions|ToFloat(Boolean) Is%20Velocity) :!class /Script/Engine.MaterialInstanceDynamic)
  (Rendering|Material|SetVectorParameterValue _advect_mid_2 "P" (Math|Conversions|ToLinearColor(Vector) (+ (/ (Variables|Default|GetForce%20Location) 500) 0.5)) :!class /Script/Engine.MaterialInstanceDynamic) ; Mouse Painting
  (Rendering|Material|SetVectorParameterValue _advect_mid_2 "V" (Math|Color|MakeColor (.x (Variables|Default|GetForce%20Velocity)) (.y (Variables|Default|GetForce%20Velocity)) (* (Math|Vector|VectorLength (Variables|Default|GetForce%20Velocity)) 0.1) 0.0) :!class /Script/Engine.MaterialInstanceDynamic)
  (Rendering|ClearRenderTarget2D Destination%20Texture)
  (Rendering|DrawMaterialtoRenderTarget Destination%20Texture _advect_mid)
  (Rendering|Material|SetScalarParameterValue _advect_mid "Force" (Math|Float|Power (Math|Float|Clamp(Float) (/ (- (Variables|Simulation|GetEmission%20Lifetime) (Actor|GetGameTimeSinceCreation)) (Variables|Simulation|GetEmission%20Lifetime))) (Variables|Simulation|GetEmission%20Curve%20Power)) :!class /Script/Engine.MaterialInstanceDynamic))

(fn CopyRenderTarget
  (params (Source%20Texture :type Texture%20Render%20Target%202D%20Object%20Reference) (Destination%20Texture :type Texture%20Render%20Target%202D%20Object%20Reference))
  (bind _velocity_preview_mid (Variables|MIDs|GetVelocity%20Preview%20MID))
  (Rendering|Material|SetTextureParameterValue _velocity_preview_mid "RT" Source%20Texture)
  (Rendering|DrawMaterialtoRenderTarget Destination%20Texture _velocity_preview_mid))

(fn Divergence
  (params (Source%20Texture :type Texture%20Render%20Target%202D%20Object%20Reference) (Desination%20Texture :type Texture%20Render%20Target%202D%20Object%20Reference))
  (Rendering|Material|SetTextureParameterValue (Variables|MIDs|GetDivergence%20MID) "Velocity" Source%20Texture)
  (Rendering|DrawMaterialtoRenderTarget Desination%20Texture (Variables|MIDs|GetDivergence%20MID)))

(fn GradientSubtraction
  (params (Source%20Velocity%20Texture :type Texture%20Render%20Target%202D%20Object%20Reference) (Source%20Pressure%20Gradient%20Texture :type Texture%20Render%20Target%202D%20Object%20Reference) (Desination%20Velocity%20Texture :type Texture%20Render%20Target%202D%20Object%20Reference))
  (bind _gradient_subtraction_mid (Variables|MIDs|GetGradient%20Subtraction%20MID))
  (Rendering|Material|SetTextureParameterValue _gradient_subtraction_mid "Velocity" Source%20Velocity%20Texture)
  (Rendering|Material|SetTextureParameterValue _gradient_subtraction_mid "Pressure" Source%20Pressure%20Gradient%20Texture)
  (Rendering|ClearRenderTarget2D Desination%20Velocity%20Texture)
  (Rendering|DrawMaterialtoRenderTarget Desination%20Velocity%20Texture (Variables|MIDs|GetGradient%20Subtraction%20MID)))

(fn PressureIteration
  (params (Pressure :type Texture%20Render%20Target%202D%20Object%20Reference) (Scratch%20RT :type Texture%20Render%20Target%202D%20Object%20Reference) (Divergence :type Texture%20Render%20Target%202D%20Object%20Reference))
  (bind _idx (LocalVariables|GetIdx))
  (bind _returnvalue (+ _idx 1))
  (bind _returnvalue_2 (Math|Integer|%25%28Integer%29 _returnvalue 2))
  (bind _returnvalue_3 (Utilities|Select :Index _returnvalue_2 :Option%200 Pressure :Option%201 Scratch%20RT))
  (Rendering|ClearRenderTarget2D Pressure)
  (Rendering|ClearRenderTarget2D Scratch%20RT)
  (Rendering|Material|SetTextureParameterValue (Variables|MIDs|GetPressure%20Iteration%20MID) "Divergence" Divergence)
  (for _index (range (Variables|Default|GetJacobi%20Iterations))
    (LocalVariables|SetIdx _index)
    (Rendering|Material|SetTextureParameterValue (Variables|MIDs|GetPressure%20Iteration%20MID) "Pressure" (Utilities|Select :Index (Math|Integer|%25%28Integer%29 (LocalVariables|GetIdx) 2) :Option%200 Pressure :Option%201 Scratch%20RT))
    (Rendering|DrawMaterialtoRenderTarget _returnvalue_3 (Variables|MIDs|GetPressure%20Iteration%20MID)))
  (Variables|Default|SetPressure%20Texture _returnvalue_3))

(fn SetMIDParameters
  (Rendering|Material|SetScalarParameterValue (Variables|MIDs|GetDivergence%20MID) "Resolution" (Math|Conversions|ToFloat(Integer) (Variables|Default|Get2D%20Resolution)) :!class /Script/Engine.MaterialInstanceDynamic)
  (Rendering|Material|SetScalarParameterValue (Variables|MIDs|GetDivergence%20MID) "FrameResolution" (/ (Math|Conversions|ToFloat(Integer) (Variables|Default|Get2D%20Resolution)) (Math|Conversions|ToFloat(Integer) (Variables|Default|GetXYFrames))) :!class /Script/Engine.MaterialInstanceDynamic)
  (Rendering|Material|SetScalarParameterValue (Variables|MIDs|GetDivergence%20MID) "XYFrames" (Math|Conversions|ToFloat(Integer) (Variables|Default|GetXYFrames)) :!class /Script/Engine.MaterialInstanceDynamic))

(fn ConstructionScript
  (Utilities|FlowControl|Sequence
    (:then_1
      (bind _returnvalue (Rendering|Material|CreateDynamicMaterialInstance "/Volumetrics/Tools/FluidSimulation/Materials/2D/M_Calculate_Divergence.M_Calculate_Divergence" "None" "None" :!class /Script/Engine.KismetMaterialLibrary))
      (Variables|MIDs|SetDivergence%20MID _returnvalue))
    (:then_2
      (bind _returnvalue_2 (Rendering|Material|CreateDynamicMaterialInstance "/Volumetrics/Tools/FluidSimulation/Materials/2D/M_Pressure_Iteration.M_Pressure_Iteration" "None" "None" :!class /Script/Engine.KismetMaterialLibrary))
      (Variables|MIDs|SetPressure%20Iteration%20MID _returnvalue_2))
    (:then_3
      (bind _returnvalue_3 (Rendering|Material|CreateDynamicMaterialInstance "/Volumetrics/Tools/FluidSimulation/Materials/2D/M_Gradient_Subtraction.M_Gradient_Subtraction" "None" "None" :!class /Script/Engine.KismetMaterialLibrary))
      (Variables|MIDs|SetGradient%20Subtraction%20MID _returnvalue_3))
    (:then_6
      (bind _returnvalue_4 (Rendering|Material|CreateDynamicMaterialInstance "/Volumetrics/Tools/FluidSimulation/Materials/Debug/M_Draw_RT.M_Draw_RT" "None" "None" :!class /Script/Engine.KismetMaterialLibrary))
      (bind _output_get (Variables|MIDs|SetPressure%20Preview%20MID _returnvalue_4))
      (Rendering|Material|SetMaterial (Variables|Default|GetPressurePreview) 0 _output_get))
    (:then_7
      (bind _returnvalue_5 (Rendering|Material|CreateDynamicMaterialInstance "/Volumetrics/Tools/FluidSimulation/Materials/Debug/M_Draw_RT.M_Draw_RT" "None" "None" :!class /Script/Engine.KismetMaterialLibrary))
      (bind _output_get_2 (Variables|MIDs|SetVelocity%20Preview%20MID _returnvalue_5))
      (Rendering|Material|SetMaterial (Variables|Default|GetVelocityPreview) 0 _output_get_2)
      (Rendering|Material|SetTextureParameterValue _output_get_2 "RT" "/Volumetrics/Content/VolumeTextures/Textures/RT_Default.RT_Default"))
    (:then_8
      (bind _returnvalue_6 (Rendering|Material|CreateDynamicMaterialInstance "/Volumetrics/Tools/FluidSimulation/Materials/2D/M_Calculate_Divergence.M_Calculate_Divergence" "None" "None" :!class /Script/Engine.KismetMaterialLibrary))
      (bind _output_get_3 (Variables|MIDs|SetDivergence%20Preview%20MID _returnvalue_6))
      (Rendering|Material|SetMaterial (Variables|Default|GetDivergence%20Preview) 0 _output_get_3)
      (Rendering|Material|SetTextureParameterValue _output_get_3 "RT" "/Volumetrics/Content/VolumeTextures/Textures/RT_Default.RT_Default"))
    (:then_9
      (bind _returnvalue_7 (Rendering|Material|CreateDynamicMaterialInstance (Variables|Default|GetVelocity%20Seed%20Material) "None" "None" :!class /Script/Engine.KismetMaterialLibrary))
      (Variables|MIDs|SetVelocity%20Seed%20MID _returnvalue_7))
    (:then_10
      (CallFunction|SetMIDParameters)
      (CallFunction|SeedPreview))))
