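; BP_FluidSimulation_01 is a Blueprint actor that runs a GPU fluid simulation in 2D or 3D by drawing materials into render targets.
; Each tick substep applies custom forces, advects velocity, computes divergence, solves pressure, subtracts its gradient, then advects color.
; Pressure solving can be multigrid across quarter, half and full resolution.
; Obstacles can be looping, mouse-driven or voxelized meshes; optional passes add light scattering, previews and raymarch display materials.
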(event Custom|Create%20RTs
  (bind _pressure_divisor (Variables|BasicSetup|GetPressure%20Divisor))
  (bind _returnvalue (Math|Conversions|ToFloat(Integer) _pressure_divisor))
  (bind _rt_size (Variables|Debug|GetRT%20Size))
  (bind _returnvalue_2 (/ _rt_size _returnvalue))
  (bind _halfres_rt_size (Variables|Debug|GetHalfres%20RT%20Size))
  (bind _quarterres_rt_size (Variables|Debug|GetQuarterRes%20RT%20Size))
  ;[1 1 1] Create RTs
  (Utilities|CheckAndCreateRT (Variables|Debug|RenderTargets|GetVelocity%20Texture) (Variables|Debug|GetRT%20Size)
    (:Out
      (Utilities|CheckAndCreateRT (Variables|Debug|RenderTargets|GetAdvected%20Velocity%20Texture) (Variables|Debug|GetRT%20Size)
        (:Out
          (Utilities|CheckAndCreateRT (Variables|Debug|RenderTargets|GetColor%20Texture) (Variables|Debug|GetRT%20Size)
            (:Out
              (Utilities|CheckAndCreateRT (Variables|Debug|RenderTargets|GetDivergence%20Texture) _returnvalue_2
                (:Out
                  (Utilities|CheckAndCreateRT (Variables|Debug|RenderTargets|GetPressure%20Texture) _returnvalue_2
                    (:Out
                      (Utilities|CheckAndCreateRT (Variables|Debug|RenderTargets|GetPressure%20Scratch%20RT) _returnvalue_2
                        (:Out
                          (Variables|Debug|SetSaved%20Resolution (Variables|Debug|GetRT%20Size))
                          (Variables|Debug|SetSavedPressureDivisor (Variables|BasicSetup|GetPressure%20Divisor))
                          (Variables|Debug|SetSaved%20Scene%20Capture%20Res (Variables|VoxelizationSetup|GetScene%20Capture%20Resolution))
                          (Utilities|FlowControl|Sequence
                            (:then_0
                              (if (and (Variables|BasicSetup|Get3D) (Variables|Simulation|Advanced|GetMulti%20Grid%20Simulation))
                                ;[1 1 1] Multigrid
                                (Utilities|CheckAndCreateRT (Variables|Debug|RenderTargets|GetHalfres%20Pressure) _halfres_rt_size
                                  (:Out
                                    (Utilities|CheckAndCreateRT (Variables|Debug|RenderTargets|GetHalfres%20Scratch) _halfres_rt_size
                                      (:Out
                                        (Utilities|CheckAndCreateRT (Variables|Debug|RenderTargets|GetQuarterres%20Pressure) _quarterres_rt_size
                                          (:Out
                                            (Utilities|CheckAndCreateRT (Variables|Debug|RenderTargets|GetQuarterres%20Scratch) _quarterres_rt_size)))))))
                                (else
                                  (Variables|Debug|RenderTargets|SetHalfres%20Pressure "")
                                  (Variables|Debug|RenderTargets|SetHalfres%20Scratch "")
                                  (Variables|Debug|RenderTargets|SetQuarterres%20Pressure "")
                                  (Variables|Debug|RenderTargets|SetQuarterres%20Scratch ""))))
                            (:then_1
                              (if (not (Utilities|Enum|Equal(Enum) (Variables|Obstacles|GetObstacle%20Mode) "NewEnumerator0"))
                                (Utilities|CheckAndCreateRT (Variables|Debug|RenderTargets|GetObstacle%20Texture) (Variables|Debug|GetRT%20Size))
                                (else
                                  (Variables|Debug|RenderTargets|SetObstacle%20Texture ""))))
                            (:then_2
                              (if (Variables|Rendering|GetCalculate%20Light%20Scattering)
                                (Utilities|CheckAndCreateRT (Variables|Debug|RenderTargets|GetLightScatter%20Texture) (Variables|Debug|GetRT%20Size))
                                (else
                                  (Variables|Debug|RenderTargets|SetLightScatter%20Texture ""))))))))))))))))))

(event Custom|SeedColor%20+%20Preview
  (bind _velocity_texture (Variables|Debug|RenderTargets|GetVelocity%20Texture))
  (bind _obstacle_texture (Variables|Debug|RenderTargets|GetObstacle%20Texture))
  ;[1 1 1] Seeding
  (Rendering|ClearRenderTarget2D (Variables|Debug|RenderTargets|GetColor%20Texture))
  (Rendering|DrawMaterialtoRenderTarget (Variables|Debug|RenderTargets|GetColor%20Texture) (Variables|Debug|MIDs|GetColor%20Seed%20MID))
  (Rendering|ClearRenderTarget2D (Variables|Debug|RenderTargets|GetAdvected%20Velocity%20Texture))
  (Rendering|ClearRenderTarget2D _velocity_texture)
  (Rendering|DrawMaterialtoRenderTarget _velocity_texture (Variables|Debug|MIDs|GetVelocity%20Seed%20MID))
  (Utilities|IsValid _obstacle_texture
    (:Is%20Valid
      (Rendering|DrawMaterialtoRenderTarget _obstacle_texture (Variables|Debug|MIDs|GetObstacle%20MID)))))

(event EventBeginPlay
  ;[1 1 1] Begin Play
  (if (Variables|Simulation|GetEnabled)
    (Utilities|FlowControl|Delay 0.05)
    (Variables|Default|SetOldObstacleX 0.05)
    (CallFunction|CreateMIDs)
    (CallFunction|SetMIDParameters)
    (CallFunction|BindRenderTargetsToMIDs)
    (CallFunction|SeedColor+Preview)
    (Variables|Default|SetOld%20O "0.350000, 0, 0")
    (Variables|Debug|SetSimulating true)
    (Actor|Tick|SetActorTickEnabled true)
    (else
      (Actor|Tick|SetActorTickEnabled false))))

(event EventTick
  (params (DeltaSeconds :type Float%20%28single-precision%29))
  ;[1 1 1] Tick
  (if (or (< (Variables|Default|GetIndex) (Variables|Default|GetMaxIterations)) (== (Variables|Default|GetMaxIterations) 0))
    (for _ (range (+ (Variables|Simulation|Advanced|GetSubSteps) 1))
      (if (or (Utilities|Enum|Equal(Enum) (Variables|Obstacles|GetObstacle%20Mode) "NewEnumerator2") (Utilities|Enum|Equal(Enum) (Variables|Obstacles|GetObstacle%20Mode) "NewEnumerator3"))
        (CallFunction|UpdateObstacles))
      (CallFunction|SimulationStep)
      (Variables|Debug|SetCurScale (* (Variables|Debug|GetCurScale) (Variables|Default|GetScale%20Per%20Frame))))))

(event Custom|Simulation%20Step
  ;[1 1 1] Simulation
  (CallFunction|ApplyCustomForces)
  (CallFunction|Advect (Variables|Debug|RenderTargets|GetVelocity%20Texture) (Variables|Debug|RenderTargets|GetVelocity%20Texture) (Variables|Debug|RenderTargets|GetAdvected%20Velocity%20Texture) true (Variables|Debug|RenderTargets|GetColor%20Texture))
  (CallFunction|Divergence (Variables|Debug|RenderTargets|GetAdvected%20Velocity%20Texture) (Variables|Debug|RenderTargets|GetDivergence%20Texture))
  (if (Variables|Simulation|Advanced|GetMulti%20Grid%20Simulation)
    (CallFunction|MultigridPressureIterations)
    (else
      (CallFunction|PressureIteration (Variables|Debug|RenderTargets|GetDivergence%20Texture) (Variables|Debug|RenderTargets|GetPressure%20Texture) (Variables|Debug|RenderTargets|GetPressure%20Scratch%20RT) 1 (Variables|Simulation|Advanced|GetFull%20Res%20Iterations) "(R=1.000000,G=1.000000,B=0.000000,A=0.000000)" (Variables|Debug|GetPressure%20RT%20Size) (Variables|Debug|GetXY%20Frames))))
  (CallFunction|GradientSubtraction (Variables|Debug|RenderTargets|GetAdvected%20Velocity%20Texture) (Variables|Debug|RenderTargets|GetPressure%20Texture) (Variables|Debug|RenderTargets|GetVelocity%20Texture))
  (CallFunction|Advect (Variables|Debug|RenderTargets|GetVelocity%20Texture) (Variables|Debug|RenderTargets|GetColor%20Texture) (Variables|Debug|RenderTargets|GetAdvected%20Velocity%20Texture) false (Variables|Debug|RenderTargets|GetColor%20Texture))
  (CallFunction|SwapRenderTargets (Variables|Debug|RenderTargets|GetColor%20Texture) (Variables|Debug|RenderTargets|GetAdvected%20Velocity%20Texture))
  (CallFunction|BindRenderTargetsToMIDs)
  (CallFunction|UpdateDynamicMIDParams)
  (Math|Integer|IncrementInt (Variables|Default|GetIndex)
    (:%20%20
      (Utilities|FlowControl|Sequence
        (:then_0
          (if (Variables|Debug|GetCopy%20RT%20Target%20to%20CB%20Asset)
            (CallFunction|CopyRenderTarget (Variables|Debug|RenderTargets|GetVelocity%20Texture) "/Volumetrics/Tools/FluidSimulation/Textures/RT_Copy.RT_Copy")))
        (:then_1
          (if (Variables|Rendering|GetCalculate%20Light%20Scattering)
            (Rendering|DrawMaterialtoRenderTarget (Variables|Debug|RenderTargets|GetLightScatter%20Texture) (Variables|Debug|MIDs|GetScatter%20Light%20MID))))))))

(event Custom|Update%20Obstacles
  (bind (_r _g _b) (Utilities|Struct|BreakLinearColor (Math|Conversions|ToLinearColor(Vector) (* (- (Variables|Default|GetOld%20O) (* (Math|Vector|RotateVectorAroundAxis "1.000000, 0, 0" (* (Math|Float|Max(Float) (* (Math|Conversions|ToFloat(Integer) (Variables|Default|GetIndex)) (Variables|Obstacles|GetSim%20Obstacle%20Hz))) 50) "0, 0,1.000000") "0.300000,0.300000, 0")) (* (.x (Variables|Debug|GetVolumeResolution)) 1)))))
  (bind _index (Variables|Default|GetIndex))
  (bind _returnvalue (Math|Conversions|ToFloat(Integer) _index))
  (bind _sim_obstacle_hz (Variables|Obstacles|GetSim%20Obstacle%20Hz))
  (bind _returnvalue_2 (* _returnvalue _sim_obstacle_hz))
  (bind _returnvalue_3 (Math|Float|Max(Float) _returnvalue_2))
  (bind _returnvalue_4 (* _returnvalue_3 50))
  (bind _returnvalue_5 (Math|Vector|RotateVectorAroundAxis "1.000000, 0, 0" _returnvalue_4 "0, 0,1.000000"))
  (bind _returnvalue_6 (* _returnvalue_5 "0.300000,0.300000, 0"))
  (bind _obstacle_mid (Variables|Debug|MIDs|GetObstacle%20MID))
  (bind _sim_obstacle_hz_2 (Variables|Obstacles|GetSim%20Obstacle%20Hz))
  (bind _obstacle_mid_2 (Variables|Debug|MIDs|GetObstacle%20MID))
  (bind (_r_2 _g_2 _b_2) (Utilities|Struct|BreakLinearColor (Math|Conversions|ToLinearColor(Vector) (* (select (Variables|Default|GetRotating%20Obstacle) (- (* (Math|Vector|RotateVectorAroundAxis "1.000000, 0, 0" (* (Math|Float|Max(Float) (* (Math|Conversions|ToFloat(Integer) (Variables|Default|GetIndex)) (/ (Variables|Obstacles|GetSim%20Obstacle%20Hz) 6e+01))) 50) "0, 0,1.000000") "0.350000,0.350000, 0") (Variables|Default|GetOld%20O)) (Math|Vector|MakeVector (* 1 (/ (Variables|Obstacles|GetSim%20Obstacle%20Hz) 6e+01)))) (* (Variables|BasicSetup|Get2D%20Resolution) 1.0)))))
  (bind _sim_obstacle_hz_3 (Variables|Obstacles|GetSim%20Obstacle%20Hz))
  (bind _returnvalue_7 (/ _sim_obstacle_hz_3 6e+01))
  (bind _index_2 (Variables|Default|GetIndex))
  (bind _returnvalue_8 (Math|Conversions|ToFloat(Integer) _index_2))
  (bind _returnvalue_9 (* _returnvalue_8 _returnvalue_7))
  (bind _returnvalue_10 (Math|Float|Max(Float) _returnvalue_9))
  (bind _returnvalue_11 (* _returnvalue_10 50))
  (bind _returnvalue_12 (Math|Vector|RotateVectorAroundAxis "1.000000, 0, 0" _returnvalue_11 "0, 0,1.000000"))
  (bind _returnvalue_13 (* _returnvalue_12 "0.350000,0.350000, 0"))
  (bind _obstacle_mid_3 (Variables|Debug|MIDs|GetObstacle%20MID))
  (bind _obstacle_mid_4 (Variables|Debug|MIDs|GetObstacle%20MID))
  ;[1 1 1] Obstacles
  (switch Utilities|FlowControl|Switch|SwitchonObstacleUpdateType (Variables|Obstacles|GetObstacle%20Mode)
    (:NewEnumerator2
      (if (Variables|BasicSetup|Get3D)
        ;; 3D Looping Obstacle
        (if (== (Variables|Default|GetIndex) 0)
          (Variables|Default|SetOld%20O _returnvalue_6))
        (Rendering|Material|SetVectorParameterValue _obstacle_mid "P" (Math|Color|MakeColor (.x (+ _returnvalue_6 "0.500000,0.500000, 0")) (.y (+ _returnvalue_6 "0.500000,0.500000, 0")) 0.2 0.1) :!class /Script/Engine.MaterialInstanceDynamic)
        (Rendering|Material|SetVectorParameterValue _obstacle_mid "V" (Math|Color|MakeColor _r _g _b 0) :!class /Script/Engine.MaterialInstanceDynamic)
        (Rendering|Material|SetScalarParameterValue _obstacle_mid_2 "Rot" (* (Math|Conversions|ToFloat(Integer) (Variables|Default|GetIndex)) _sim_obstacle_hz_2) :!class /Script/Engine.MaterialInstanceDynamic)
        (Rendering|Material|SetScalarParameterValue _obstacle_mid_2 "RotRate" _sim_obstacle_hz_2 :!class /Script/Engine.MaterialInstanceDynamic)
        (Variables|Default|SetOld%20O _returnvalue_6)
        (Variables|Default|SetOld%20Obstacle%20Radius (Variables|Default|GetObstacle%20Radius))
        (Rendering|ClearRenderTarget2D (Variables|Debug|RenderTargets|GetObstacle%20Texture))
        (Rendering|DrawMaterialtoRenderTarget (Variables|Debug|RenderTargets|GetObstacle%20Texture) (Variables|Debug|MIDs|GetObstacle%20MID))
        (else
          ;; Looping Obstacle
          (Rendering|Material|SetVectorParameterValue _obstacle_mid_3 "P" (Math|Color|MakeColor (.x (select (Variables|Default|GetRotating%20Obstacle) (+ _returnvalue_13 "0.500000,0.500000, 0") (+ (Math|Vector|MakeVector (Math|Float|Fraction _returnvalue_10)) "0.000000,0.750000, 0"))) (.y (select (Variables|Default|GetRotating%20Obstacle) (+ _returnvalue_13 "0.500000,0.500000, 0") (+ (Math|Vector|MakeVector (Math|Float|Fraction _returnvalue_10)) "0.000000,0.750000, 0"))) 0.0 0.1) :!class /Script/Engine.MaterialInstanceDynamic)
          (Rendering|Material|SetVectorParameterValue _obstacle_mid_3 "V" (Math|Color|MakeColor _r_2 _g_2 _b_2 0) :!class /Script/Engine.MaterialInstanceDynamic)
          (Variables|Default|SetOld%20O _returnvalue_13)
          (Variables|Default|SetOld%20Obstacle%20Radius (Variables|Default|GetObstacle%20Radius))
          (Rendering|DrawMaterialtoRenderTarget (Variables|Debug|RenderTargets|GetObstacle%20Texture) (Variables|Debug|MIDs|GetObstacle%20MID)))))
    (:NewEnumerator3
      ;; Mouse Obstacle
      (Rendering|Material|SetVectorParameterValue _obstacle_mid_4 "P" (Math|Color|MakeColor (.x (+ (/ (Variables|Default|GetObstacle%20Location) 500) 0.5)) (.y (+ (/ (Variables|Default|GetObstacle%20Location) 500) 0.5)) 0.0 (Variables|Default|GetObstacle%20Radius)) :!class /Script/Engine.MaterialInstanceDynamic)
      (Rendering|Material|SetVectorParameterValue _obstacle_mid_4 "V" (Math|Color|MakeColor (.x (Variables|Default|GetForce%20Velocity)) (.y (Variables|Default|GetForce%20Velocity)) 0.0 (* (- (Variables|Default|GetOld%20Obstacle%20Radius) (Variables|Default|GetObstacle%20Radius)) 500)) :!class /Script/Engine.MaterialInstanceDynamic)
      (Variables|Default|SetOld%20Obstacle%20Radius (Variables|Default|GetObstacle%20Radius))
      (Rendering|DrawMaterialtoRenderTarget (Variables|Debug|RenderTargets|GetObstacle%20Texture) (Variables|Debug|MIDs|GetObstacle%20MID)))))

(event Custom|ToggleVelocity
  (Rendering|Material|SetMaterial (Variables|Default|GetS_Box_InsideOut_01) 0 (Variables|Debug|MIDs|GetRaymarch%20Velocity%20MID)))

(event Custom|Downsample%20Test
  (Rendering|ClearRenderTarget2D "")
  (CallFunction|ReSampleRT :XY%20Frames%20to%20Sample "(X=8.000000,Y=8.000000)" :XY%20Frames%20to%20Encode "(X=8.000000,Y=4.000000)" :UVScale "(X=1.000000,Y=1.000000)" :VolumeResolution "0, 0, 0" :RT%20Size "(X=2048.000000,Y=1024.000000)"))

(event Custom|Upsample%20Test
  (CallFunction|ReSampleRT :XY%20Frames%20to%20Sample "(X=4.000000,Y=4.000000)" :XY%20Frames%20to%20Encode "(X=8.000000,Y=8.000000)" :UVScale "(X=1.000000,Y=1.000000)" :VolumeResolution "0, 0, 0" :RT%20Size (Variables|Debug|GetRT%20Size)))

(event Custom|ReUp%20Sample
  (Rendering|DrawMaterialtoRenderTarget "")
  (Rendering|ClearRenderTarget2D "")
  (CallFunction|ReSampleRT :XY%20Frames%20to%20Sample (Variables|Debug|GetXY%20Frames) :XY%20Frames%20to%20Encode (Variables|Debug|GetHalfRes%20XY%20Frames) :UVScale "(X=1.000000,Y=1.000000)" :VolumeResolution "128.000000,128.000000,128.000000" :RT%20Size (Variables|Debug|GetRT%20Size))
  (Rendering|ClearRenderTarget2D "")
  (CallFunction|ReSampleRT :XY%20Frames%20to%20Sample (Variables|Debug|GetHalfRes%20XY%20Frames) :XY%20Frames%20to%20Encode (Variables|Debug|GetQuarterRes%20XY%20Frames) :UVScale "(X=1.000000,Y=1.000000)" :VolumeResolution "64.000000,64.000000,64.000000" :RT%20Size (Variables|Debug|GetRT%20Size)))

(fn Advect
  (params (Source%20Velocity :type Texture%20Render%20Target%202D%20Object%20Reference%20%28by%20ref%29) (Source%20Texture :type Texture%20Render%20Target%202D%20Object%20Reference%20%28by%20ref%29) (Destination%20Texture :type Texture%20Render%20Target%202D%20Object%20Reference%20%28by%20ref%29) (Is%20Velocity :type Boolean) (Color%20Texture :type Texture%20Render%20Target%202D%20Object%20Reference%20%28by%20ref%29))
  (bind _advect_mid (Variables|Debug|MIDs|GetAdvect%20MID))
  (bind _advect_color_mid (Variables|Debug|MIDs|GetAdvect%20Color%20MID))
  (bind _returnvalue (select Is%20Velocity _advect_mid _advect_color_mid))
  (Rendering|Material|SetTextureParameterValue _returnvalue "Color" Color%20Texture)
  (Rendering|Material|SetTextureParameterValue _returnvalue "Velocity" Source%20Velocity)
  (Rendering|Material|SetTextureParameterValue _returnvalue "Source" Source%20Texture)
  (Rendering|Material|SetScalarParameterValue _returnvalue "IsVelocity" (Math|Conversions|ToFloat(Boolean) Is%20Velocity) :!class /Script/Engine.MaterialInstanceDynamic)
  (Rendering|Material|SetVectorParameterValue _returnvalue "P" (Math|Conversions|ToLinearColor(Vector) (- (/ (.location (Variables|SeedColorSetup|GetP)) 256) "-0.500000,-0.500000,-0.500000")) :!class /Script/Engine.MaterialInstanceDynamic) ; Mouse Painting
  (Rendering|Material|SetScalarParameterValue _returnvalue "Color" (Math|Conversions|ToFloat(Boolean) (Variables|Default|GetRightClick)) :!class /Script/Engine.MaterialInstanceDynamic)
  (Rendering|ClearRenderTarget2D Destination%20Texture)
  (if (Variables|Default|GetRK)
    (Rendering|DrawMaterialtoRenderTarget Destination%20Texture (Variables|Debug|MIDs|GetRK%20MID))
    (CallFunction|CopyRenderTarget Destination%20Texture Source%20Texture))
  (Rendering|ClearRenderTarget2D Destination%20Texture)
  (Rendering|Material|SetScalarParameterValue _returnvalue "Emission Strength" (* (Variables|Simulation|GetEmission%20Multiplier) (Math|Curves|GetFloatValue (Variables|Simulation|GetEmission%20Curve) (Actor|GetGameTimeSinceCreation))) :!class /Script/Engine.MaterialInstanceDynamic)
  (Rendering|Material|SetScalarParameterValue _returnvalue "Emission Temperature" (Variables|Simulation|GetEmission%20Temperature%20Multiplier) :!class /Script/Engine.MaterialInstanceDynamic)
  (Rendering|DrawMaterialtoRenderTarget Destination%20Texture _returnvalue))

(fn ApplyCustomForces
  (bind _forcemids (Variables|Debug|MIDs|GetForceMIDs))
  (bind _custom_forces (Variables|Forces|GetCustom%20Forces))
  (bind _keys (Utilities|Map|Keys _custom_forces))
  (for _array_element _keys
    (if (and (.ReturnValue (Utilities|Map|Find _forcemids _array_element)) (Utilities|Struct|BreakFluidSimCustomForce (Utilities|Map|Find _custom_forces _array_element)))
      (Utilities|IsValid (Utilities|Struct|BreakFluidSimForceMIDs (Utilities|Map|Find _forcemids _array_element))
        (:Is%20Valid
          (CallFunction|ApplySingleForce (Utilities|Struct|BreakFluidSimForceMIDs (Utilities|Map|Find _forcemids _array_element)) (Variables|Debug|RenderTargets|GetColor%20Texture) (Variables|Debug|RenderTargets|GetVelocity%20Texture) (Variables|Debug|RenderTargets|GetAdvected%20Velocity%20Texture)) ; Apply Velocity
          (CallFunction|SwapRenderTargets (Variables|Debug|RenderTargets|GetVelocity%20Texture) (Variables|Debug|RenderTargets|GetAdvected%20Velocity%20Texture)))
        (:Is%20Not%20Valid
          _))
      (Utilities|IsValid (.ColorMID_4_C7CC5CFA43627B4C5D0617AEDEE80F8F (Utilities|Struct|BreakFluidSimForceMIDs (Utilities|Map|Find _forcemids _array_element)))
        (:Is%20Valid
          (CallFunction|ApplySingleForce (.ColorMID_4_C7CC5CFA43627B4C5D0617AEDEE80F8F (Utilities|Struct|BreakFluidSimForceMIDs (Utilities|Map|Find _forcemids _array_element))) (Variables|Debug|RenderTargets|GetColor%20Texture) (Variables|Debug|RenderTargets|GetColor%20Texture) (Variables|Debug|RenderTargets|GetAdvected%20Velocity%20Texture)) ; Apply Color
          (CallFunction|SwapRenderTargets (Variables|Debug|RenderTargets|GetColor%20Texture) (Variables|Debug|RenderTargets|GetAdvected%20Velocity%20Texture)))))))

(fn ApplySingleForce
  (params (MID :type Material%20Instance%20Dynamic%20Object%20Reference) (Color%20Texture :type Texture%20Render%20Target%202D%20Object%20Reference%20%28by%20ref%29) (Source%20Texture :type Texture%20Render%20Target%202D%20Object%20Reference%20%28by%20ref%29) (Destination%20Texture :type Texture%20Render%20Target%202D%20Object%20Reference%20%28by%20ref%29))
  (Rendering|Material|SetTextureParameterValue MID "Color" Color%20Texture)
  (Rendering|Material|SetTextureParameterValue MID "Source" Source%20Texture)
  (Rendering|ClearRenderTarget2D Destination%20Texture)
  (Rendering|DrawMaterialtoRenderTarget Destination%20Texture MID))

(fn BindRenderTargetstoMIDs
  (bind (_velocitymid_2_b4527a8841e85f646c62809844f07f3d _colormid_4_c7cc5cfa43627b4c5d0617aedee80f8f) (Utilities|Struct|BreakFluidSimForceMIDs (Utilities|Map|Find (Variables|Debug|MIDs|GetForceMIDs) "SkelMesh")))
  (bind (_velocitymid_2_b4527a8841e85f646c62809844f07f3d_2 _colormid_4_c7cc5cfa43627b4c5d0617aedee80f8f_2) (Utilities|Struct|BreakFluidSimForceMIDs (Utilities|Map|Find (Variables|Debug|MIDs|GetForceMIDs) "SkelMesh")))
  (bind _forcemids (Variables|Debug|MIDs|GetForceMIDs))
  (bind (_value _returnvalue) (Utilities|Map|Find _forcemids "SkelMesh"))
  (bind _voxelization_bp (Variables|Obstacles|GetVoxelization%20BP))
  (bind _jumpfloodcomponent_3d (Variables|/Volumetrics/Tools/VoxelizeMeshes/Blueprints/BP_VoxelizeMeshes.BP_VoxelizeMeshes_C|GetJumpFloodComponent_3D _voxelization_bp))
  (bind _lastindex (Variables|/Volumetrics/Tools/JumpFlooding/Blueprints/JumpFloodComponent_3D.JumpFloodComponent_3D_C|GetLastIndex _jumpfloodcomponent_3d))
  (bind _returnvalue_2 (Math|Integer|%25%28Integer%29 _lastindex 2))
  (bind _returnvalue_3 (== _returnvalue_2 0))
  (bind _jumpfloodcomponent_3d_2 (Variables|/Volumetrics/Tools/VoxelizeMeshes/Blueprints/BP_VoxelizeMeshes.BP_VoxelizeMeshes_C|GetJumpFloodComponent_3D _voxelization_bp))
  (bind _rt_b (Variables|/Volumetrics/Tools/JumpFlooding/Blueprints/JumpFloodComponent_3D.JumpFloodComponent_3D_C|GetRT%20B _jumpfloodcomponent_3d_2))
  (bind _rt_a (Variables|/Volumetrics/Tools/JumpFlooding/Blueprints/JumpFloodComponent_3D.JumpFloodComponent_3D_C|GetRT%20A _jumpfloodcomponent_3d_2))
  (bind _returnvalue_4 (select _returnvalue_3 _rt_b _rt_a))
  (bind _obstacle_texture (Variables|Debug|RenderTargets|GetObstacle%20Texture))
  ;[1 1 1] Bind Render Targets
  (Rendering|Material|SetTextureParameterValue (Variables|Debug|MIDs|GetVelocity%20Seed%20MID) "Color" (Variables|Debug|RenderTargets|GetColor%20Texture))
  (Rendering|Material|SetTextureParameterValue (Variables|Debug|MIDs|GetVelocity%20Preview%20MID) "Velocity" (select (Variables|Debug|GetShow%20Obstacle%20In%20Preview) (Variables|Debug|RenderTargets|GetObstacle%20Texture) (Variables|Debug|RenderTargets|GetVelocity%20Texture)))
  (Rendering|Material|SetTextureParameterValue (Variables|Debug|MIDs|GetRaymarch%20Color%20and%20Opacity%20MID) "Pressure" (Variables|Debug|RenderTargets|GetPressure%20Texture))
  ;[1 1 1] Mesh Force Texture Binding
  (if (Utilities|Enum|Equal(Enum) (Variables|Obstacles|GetObstacle%20Mode) "NewEnumerator4")
    (Utilities|IsValid (Variables|Obstacles|GetVoxelization%20BP)
      (:Is%20Valid
        (Utilities|FlowControl|Sequence
          (:then_0
            (if (Utilities|Enum|Equal(Enum) (Variables|/Volumetrics/Tools/VoxelizeMeshes/Blueprints/BP_VoxelizeMeshes.BP_VoxelizeMeshes_C|GetVolume%20Resolution (Variables|Obstacles|GetVoxelization%20BP)) (Variables|BasicSetup|GetVolume%20Resolution))
              (Variables|Debug|RenderTargets|SetObstacle%20Texture (Variables|/Volumetrics/Tools/VoxelizeMeshes/Blueprints/BP_VoxelizeMeshes.BP_VoxelizeMeshes_C|GetVoxelizationRTA (Variables|Obstacles|GetVoxelization%20BP)))
              (else
                (CallFunction|ReSampleRT :XY%20Frames%20to%20Sample (Variables|/Volumetrics/Tools/VoxelizeMeshes/Blueprints/BP_VoxelizeMeshes.BP_VoxelizeMeshes_C|GetXY%20Frames (Variables|Obstacles|GetVoxelization%20BP)) :XY%20Frames%20to%20Encode (Variables|Debug|GetXY%20Frames) :UVScale "(X=1.000000,Y=1.000000)" :VolumeResolution (Variables|Debug|GetVolumeResolution) :Source%20RT (Variables|/Volumetrics/Tools/VoxelizeMeshes/Blueprints/BP_VoxelizeMeshes.BP_VoxelizeMeshes_C|GetVoxelizationRTA (Variables|Obstacles|GetVoxelization%20BP)) :Destination%20RT (Variables|Debug|RenderTargets|GetObstacle%20Texture) :RT%20Size (Variables|Debug|GetRT%20Size)))))
          (:then_1
            (if _returnvalue
              (Rendering|Material|SetTextureParameterValue _velocitymid_2_b4527a8841e85f646c62809844f07f3d "SkeletalMeshTexture" _obstacle_texture)
              (Rendering|Material|SetTextureParameterValue _colormid_4_c7cc5cfa43627b4c5d0617aedee80f8f "SkeletalMeshTexture" _obstacle_texture)
              (if (Variables|/Volumetrics/Tools/VoxelizeMeshes/Blueprints/BP_VoxelizeMeshes.BP_VoxelizeMeshes_C|GetGenerate%20Distance%20FIeld (Variables|Obstacles|GetVoxelization%20BP))
                (Rendering|Material|SetTextureParameterValue _velocitymid_2_b4527a8841e85f646c62809844f07f3d_2 "Gradient" _returnvalue_4)
                (Rendering|Material|SetTextureParameterValue _colormid_4_c7cc5cfa43627b4c5d0617aedee80f8f_2 "Gradient" _returnvalue_4)))))))))

(fn CopyRenderTarget
  (params (Source%20Texture :type Texture%20Render%20Target%202D%20Object%20Reference) (Destination%20Texture :type Texture%20Render%20Target%202D%20Object%20Reference))
  (Rendering|Material|SetTextureParameterValue (Variables|Debug|MIDs|GetCopy%20RT%20MID) "RT" Source%20Texture)
  (Rendering|ClearRenderTarget2D Destination%20Texture)
  (Rendering|DrawMaterialtoRenderTarget Destination%20Texture (Variables|Debug|MIDs|GetCopy%20RT%20MID)))

(fn CreateMIDs
  (bind _advect_mid (Variables|Debug|MIDs|GetAdvect%20MID))
  (bind _custom_forces (Variables|Forces|GetCustom%20Forces))
  (bind _advect_color_mid (Variables|Debug|MIDs|GetAdvect%20Color%20MID))
  (bind _forcemids (Variables|Debug|MIDs|GetForceMIDs))
  (Utilities|FlowControl|Sequence
    (:then_0
      (Class|VolumetricsBPLIbrary|CreateAndAssignMID _advect_mid (select (Variables|BasicSetup|Get3D) (Variables|Simulation|Get3D%20Advection%20Material) (Variables|Simulation|Get2D%20Advection%20Material)))
      (Rendering|Material|SetScalarParameterValue _advect_mid "TimeStep" (Variables|Simulation|GetTime%20Step) :!class /Script/Engine.MaterialInstanceDynamic)
      (Class|VolumetricsBPLIbrary|CreateAndAssignMID _advect_color_mid (select (Variables|BasicSetup|Get3D) (Variables|Simulation|Get3D%20Color%20Advection) (Variables|Simulation|Get2D%20Advection%20Material)))
      (Rendering|Material|SetScalarParameterValue _advect_color_mid "TimeStep" (Variables|Simulation|GetTime%20Step) :!class /Script/Engine.MaterialInstanceDynamic))
    (:then_1
      (Class|VolumetricsBPLIbrary|CreateAndAssignMID (Variables|Debug|MIDs|GetDivergence%20MID) (select (Variables|BasicSetup|Get3D) "/Volumetrics/Tools/FluidSimulation/Materials/3D/M_Calculate_Divergence_3D.M_Calculate_Divergence_3D" "/Volumetrics/Tools/FluidSimulation/Materials/2D/M_Calculate_Divergence.M_Calculate_Divergence")))
    (:then_2
      (Class|VolumetricsBPLIbrary|CreateAndAssignMID (Variables|Debug|MIDs|GetPressure%20Iteration%20MID) (select (Variables|BasicSetup|Get3D) "/Volumetrics/Tools/FluidSimulation/Materials/3D/M_Pressure_Iteration_3D.M_Pressure_Iteration_3D" "/Volumetrics/Tools/FluidSimulation/Materials/2D/M_Pressure_Iteration.M_Pressure_Iteration")))
    (:then_3
      (Class|VolumetricsBPLIbrary|CreateAndAssignMID (Variables|Debug|MIDs|GetGradient%20Subtraction%20MID) (select (Variables|BasicSetup|Get3D) "/Volumetrics/Tools/FluidSimulation/Materials/3D/M_Gradient_Subtraction_3D.M_Gradient_Subtraction_3D" "/Volumetrics/Tools/FluidSimulation/Materials/2D/M_Gradient_Subtraction.M_Gradient_Subtraction")))
    (:then_4
      (Class|VolumetricsBPLIbrary|CreateAndAssignMID (Variables|Debug|MIDs|GetBoundary%20Condition%20MID) "/Volumetrics/Tools/FluidSimulation/Materials/2D/M_Boundary_Condition.M_Boundary_Condition"))
    (:then_5
      (Class|VolumetricsBPLIbrary|CreateAndAssignMID (Variables|Debug|MIDs|GetObstacle%20MID) (select (Variables|BasicSetup|Get3D) (Variables|Obstacles|Get3D%20Obstacle%20Material) (Variables|Obstacles|Get2D%20Obstacle%20Material))))
    (:then_6
      (Utilities|FlowControl|Sequence
        (:then_0
          (Class|VolumetricsBPLIbrary|CreateAndAssignMID (Variables|Debug|MIDs|GetColor%20Preview%20MID) (select (Variables|BasicSetup|Get3D) "/Volumetrics/Tools/FluidSimulation/Materials/Debug/M_Preview_Color_3d.M_Preview_Color_3d" "/Volumetrics/Tools/FluidSimulation/Materials/Debug/M_Preview_Color.M_Preview_Color"))
          (Rendering|Material|SetMaterial (Variables|Default|GetAlpha%20Preview) 0 (Variables|Debug|MIDs|GetColor%20Preview%20MID)))
        (:then_1
          (Class|VolumetricsBPLIbrary|CreateAndAssignMID (Variables|Debug|MIDs|GetVelocity%20Preview%20MID) "/Volumetrics/Tools/FluidSimulation/Materials/Debug/M_Preview_Velocity.M_Preview_Velocity")
          (Rendering|Material|SetMaterial (Variables|Default|GetVelocity%20Preview) 0 (Variables|Debug|MIDs|GetVelocity%20Preview%20MID)))
        (:then_2
          (Class|VolumetricsBPLIbrary|CreateAndAssignMID (Variables|Debug|MIDs|GetRaymarch%20Color%20and%20Opacity%20MID) (Variables|Rendering|GetDefault%20Rendering%20Material)))
        (:then_3
          (Class|VolumetricsBPLIbrary|CreateAndAssignMID (Variables|Debug|MIDs|GetVelocity%20Seed%20MID) (Variables|SeedVelocitySetup|GetVelocity%20Seed%20Material))
          (Class|VolumetricsBPLIbrary|CreateAndAssignMID (Variables|Debug|MIDs|GetColor%20Seed%20MID) (Variables|SeedColorSetup|GetColor%20Seed%20Material)))
        (:then_4
          (Class|VolumetricsBPLIbrary|CreateAndAssignMID (Variables|Debug|MIDs|GetCopy%20RT%20MID) "/Volumetrics/Tools/FluidSimulation/Materials/Debug/M_Draw_RT.M_Draw_RT"))
        (:then_5
          (Class|VolumetricsBPLIbrary|CreateAndAssignMID (Variables|Default|GetResample%20RT%20MID) "/Volumetrics/Tools/FluidSimulation/Materials/3D/M_Resample_RT.M_Resample_RT"))
        (:then_6
          (Class|VolumetricsBPLIbrary|CreateAndAssignMID (Variables|Debug|MIDs|GetRK%20MID) "/Volumetrics/Tools/FluidSimulation/Materials/2D/M_RK.M_RK"))
        (:then_7
          (Class|VolumetricsBPLIbrary|CreateAndAssignMID (Variables|Debug|MIDs|GetRaymarch%20Normals%20MID) (Variables|Rendering|GetNormals%20Rendering%20Material)))
        (:then_8
          (Utilities|FlowControl|Sequence
            (:then_0
              (Class|VolumetricsBPLIbrary|CreateAndAssignMID (Variables|Debug|MIDs|GetRaymarch%20Velocity%20MID) (Variables|Rendering|GetVelocity%20Rendering%20Material))
              (Class|VolumetricsBPLIbrary|CreateAndAssignMID (Variables|Debug|MIDs|GetRaymarch%20SH%20MID) (Variables|Rendering|GetSpherical%20Harmonic%20Material))
              (Class|VolumetricsBPLIbrary|CreateAndAssignMID (Variables|Debug|MIDs|GetRaymarch%20Temperature%20MID) (Variables|Rendering|GetTemperature%20Rendering%20Material))
              (CallFunction|SetVolumeDisplayMode (Variables|Rendering|GetDisplay%20Material)))
            (:then_1
              (Class|VolumetricsBPLIbrary|CreateAndAssignMID (Variables|Debug|MIDs|GetScatter%20Light%20MID) "/Volumetrics/Tools/FluidSimulation/Materials/3D/Rendering/M_Volume_Light_Scatter.M_Volume_Light_Scatter"))
            (:then_2
              ;[0.1302 0.1302 0.1302] Custom Forces
              (if (> (Utilities|Map|Length (Variables|Forces|GetCustom%20Forces)) 0)
                (bind _keys (Utilities|Map|Keys (Variables|Debug|MIDs|GetForceMIDs)))
                (for _array_element _keys
                  (if (not (.ReturnValue (Utilities|Map|Find (Variables|Forces|GetCustom%20Forces) _array_element)))
                    (Utilities|Map|Remove (Variables|Debug|MIDs|GetForceMIDs) _array_element)))
                (bind _keys_2 (Utilities|Map|Keys (Variables|Forces|GetCustom%20Forces)))
                (for _array_element_2 _keys_2
                  (if (.ReturnValue (Utilities|Map|Find _forcemids _array_element_2))
                    (LocalVariables|SetLocalMID (Utilities|Struct|BreakFluidSimForceMIDs (Utilities|Map|Find _forcemids _array_element_2)))
                    (LocalVariables|SetLocalMID2 (.ColorMID_4_C7CC5CFA43627B4C5D0617AEDEE80F8F (Utilities|Struct|BreakFluidSimForceMIDs (Utilities|Map|Find _forcemids _array_element_2))))
                    (else
                      (LocalVariables|SetLocalMID 0)
                      (LocalVariables|SetLocalMID2 0)))
                  (Utilities|IsValid (.MaterialVelocity_17_B454C6464D1F301078BC34929299E8DD (Utilities|Struct|BreakFluidSimCustomForce (Utilities|Map|Find _custom_forces _array_element_2)))
                    (:Is%20Valid
                      (Class|VolumetricsBPLIbrary|CreateAndAssignMID (LocalVariables|GetLocalMID) (.MaterialVelocity_17_B454C6464D1F301078BC34929299E8DD (Utilities|Struct|BreakFluidSimCustomForce (Utilities|Map|Find _custom_forces _array_element_2)))))
                    (:Is%20Not%20Valid
                      _))
                  (Utilities|IsValid (.MaterialColor_18_A5FD7AC74B426B20784BD18818571CB1 (Utilities|Struct|BreakFluidSimCustomForce (Utilities|Map|Find _custom_forces _array_element_2)))
                    (:Is%20Valid
                      (Class|VolumetricsBPLIbrary|CreateAndAssignMID (LocalVariables|GetLocalMID2) (.MaterialColor_18_A5FD7AC74B426B20784BD18818571CB1 (Utilities|Struct|BreakFluidSimCustomForce (Utilities|Map|Find _custom_forces _array_element_2)))))
                    (:Is%20Not%20Valid
                      _))
                  (Utilities|Map|Add (Variables|Debug|MIDs|GetForceMIDs) _array_element_2 (Utilities|Struct|MakeFluidSimForceMIDs (LocalVariables|GetLocalMID) (LocalVariables|GetLocalMID2))))
                (else
                  (Utilities|Map|Clear (Variables|Debug|MIDs|GetForceMIDs)))))))))))

(fn DebugPreviewSetup
  (Rendering|SetVisibility (Variables|Default|GetS_Box_InsideOut_01) (Variables|Debug|GetPreview%20Volume) false)
  (Rendering|SetVisibility (Variables|Default|GetAlpha%20Preview) (Variables|Debug|GetShow%20Color%20Texture) false)
  (Rendering|SetVisibility (Variables|Default|GetVelocity%20Preview) (Variables|Debug|GetShow%20Velocity%20Texture) false))

(fn Divergence
  (params (Source%20Texture :type Texture%20Render%20Target%202D%20Object%20Reference) (Desination%20Texture :type Texture%20Render%20Target%202D%20Object%20Reference))
  (Rendering|Material|SetTextureParameterValue (Variables|Debug|MIDs|GetDivergence%20MID) "Velocity" Source%20Texture)
  (Rendering|DrawMaterialtoRenderTarget Desination%20Texture (Variables|Debug|MIDs|GetDivergence%20MID)))

(fn GradientSubtraction
  (params (Source%20Velocity%20Texture :type Texture%20Render%20Target%202D%20Object%20Reference) (Source%20Pressure%20Gradient%20Texture :type Texture%20Render%20Target%202D%20Object%20Reference) (Desination%20Velocity%20Texture :type Texture%20Render%20Target%202D%20Object%20Reference))
  (bind _gradient_subtraction_mid (Variables|Debug|MIDs|GetGradient%20Subtraction%20MID))
  (Rendering|Material|SetTextureParameterValue _gradient_subtraction_mid "Velocity" Source%20Velocity%20Texture)
  (Rendering|Material|SetTextureParameterValue _gradient_subtraction_mid "Pressure" Source%20Pressure%20Gradient%20Texture)
  (Rendering|ClearRenderTarget2D Desination%20Velocity%20Texture)
  (Rendering|DrawMaterialtoRenderTarget Desination%20Velocity%20Texture (Variables|Debug|MIDs|GetGradient%20Subtraction%20MID)))

(fn MultigridPressureIterations
  (bind (_x _y) (Math|Vector2D|BreakVector2D (- (Variables|Debug|GetRT%20Size) (Variables|Debug|GetHalfres%20RT%20Size))))
  (bind (_x_2 _y_2) (Math|Vector2D|BreakVector2D (Variables|Debug|GetHalfres%20RT%20Size)))
  (bind (_x_3 _y_3) (Math|Vector2D|BreakVector2D (- (Variables|Debug|GetRT%20Size) (+ (* (Variables|Debug|GetHalfres%20RT%20Size) "(X=0.000000,Y=1.000000)") (Variables|Debug|GetQuarterRes%20RT%20Size)))))
  (bind (_x_4 _y_4) (Math|Vector2D|BreakVector2D (Variables|Debug|GetRT%20Size)))
  (bind (_x_5 _y_5) (Math|Vector2D|BreakVector2D (Variables|Debug|GetRT%20Size)))
  (bind (_x_6 _y_6) (Math|Vector2D|BreakVector2D (Variables|Debug|GetQuarterRes%20RT%20Size)))
  (CallFunction|ReSampleRT :XY%20Frames%20to%20Sample (Variables|Debug|GetXY%20Frames) :XY%20Frames%20to%20Encode (Variables|Debug|GetHalfRes%20XY%20Frames) :UVScale "(X=1.000000,Y=1.000000)" :VolumeResolution (/ (Variables|Debug|GetVolumeResolution) 2) :Source%20RT (Variables|Debug|RenderTargets|GetPressure%20Texture) :Destination%20RT (Variables|Debug|RenderTargets|GetHalfres%20Pressure) :RT%20Size (Variables|Debug|GetHalfres%20RT%20Size)) ; Downsample Pressure from Previous Pass
  (CallFunction|ReSampleRT :XY%20Frames%20to%20Sample (Variables|Debug|GetHalfRes%20XY%20Frames) :XY%20Frames%20to%20Encode (Variables|Debug|GetQuarterRes%20XY%20Frames) :UVScale "(X=1.000000,Y=1.000000)" :VolumeResolution (/ (Variables|Debug|GetVolumeResolution) 4) :Source%20RT (Variables|Debug|RenderTargets|GetHalfres%20Pressure) :Destination%20RT (Variables|Debug|RenderTargets|GetQuarterres%20Pressure) :RT%20Size (Variables|Debug|GetQuarterRes%20RT%20Size)) ; Downsample Pressure from Previous Pass
  (if (> (Variables|Simulation|Advanced|GetQuarter%20Res%20Iterations) 0)
    ;[1 1 1] Quarter Res Iterations
    (CallFunction|PressureIteration (Variables|Debug|RenderTargets|GetDivergence%20Texture) (Variables|Debug|RenderTargets|GetQuarterres%20Pressure) (Variables|Debug|RenderTargets|GetQuarterres%20Scratch) 4 (Variables|Simulation|Advanced|GetQuarter%20Res%20Iterations) (Utilities|Struct|MakeLinearColor (/ _x_6 _x_5) (/ _y_6 _y_5) (/ _x_3 _x_5) (/ _y_3 _y_5)) (Variables|Debug|GetQuarterRes%20RT%20Size) (Variables|Debug|GetQuarterRes%20XY%20Frames)))
  ;[1 1 1] Upscale Pressure to Half Res
  (CallFunction|ReSampleRT :XY%20Frames%20to%20Sample (Variables|Debug|GetQuarterRes%20XY%20Frames) :XY%20Frames%20to%20Encode (Variables|Debug|GetHalfRes%20XY%20Frames) :UVScale "(X=1.000000,Y=1.000000)" :VolumeResolution (/ (Variables|Debug|GetVolumeResolution) 4) :Source%20RT (Variables|Debug|RenderTargets|GetQuarterres%20Pressure) :Destination%20RT (Variables|Debug|RenderTargets|GetHalfres%20Pressure) :RT%20Size (Variables|Debug|GetHalfres%20RT%20Size))
  (if (> (Variables|Simulation|Advanced|GetHalf%20Res%20Iterations) 0)
    ;[1 1 1] Half Res Iterations
    (CallFunction|PressureIteration (Variables|Debug|RenderTargets|GetDivergence%20Texture) (Variables|Debug|RenderTargets|GetHalfres%20Pressure) (Variables|Debug|RenderTargets|GetHalfres%20Scratch) 2 (Variables|Simulation|Advanced|GetHalf%20Res%20Iterations) (Utilities|Struct|MakeLinearColor (/ _x_2 _x_4) (/ _y_2 _y_4) (/ _x _x_4) (/ _y _y_4)) (Variables|Debug|GetHalfres%20RT%20Size) (Variables|Debug|GetHalfRes%20XY%20Frames)))
  ;[1 1 1] Upscale Pressure to Full Res
  (CallFunction|ReSampleRT (Variables|Debug|GetHalfRes%20XY%20Frames) (Variables|Debug|GetXY%20Frames) "(X=1.000000,Y=1.000000)" (/ (Variables|Debug|GetVolumeResolution) 2) (Variables|Debug|RenderTargets|GetHalfres%20Pressure) (Variables|Debug|RenderTargets|GetPressure%20Texture) "(X=0.000000,Y=0.000000)" (Variables|Debug|GetRT%20Size))
  ;[1 1 1] Full Res Iterations
  (CallFunction|PressureIteration (Variables|Debug|RenderTargets|GetDivergence%20Texture) (Variables|Debug|RenderTargets|GetPressure%20Texture) (Variables|Debug|RenderTargets|GetPressure%20Scratch%20RT) 1 (Variables|Simulation|Advanced|GetFull%20Res%20Iterations) (Utilities|Struct|MakeLinearColor 1 1) (Variables|Debug|GetRT%20Size) (Variables|Debug|GetXY%20Frames)))

(fn PressureIteration
  (params (Divergence :type Texture%20Render%20Target%202D%20Object%20Reference) (Pressure :type Texture%20Render%20Target%202D%20Object%20Reference%20%28by%20ref%29) (Scratch%20RT :type Texture%20Render%20Target%202D%20Object%20Reference%20%28by%20ref%29) (Divisor :type Float%20%28double-precision%29) (Iterations :type Integer) (UV%20Scale%20and%20Offset :type Linear%20Color%20Structure) (RT%20Size :type Vector%202D%20Structure) (XY%20Frames :type Vector%202D%20Structure))
  (bind (_x _y) (Math|Vector2D|BreakVector2D XY%20Frames))
  (bind (_x_2 _y_2) (Math|Vector2D|BreakVector2D (Variables|Debug|GetXY%20Frames)))
  (bind (_x_3 _y_3) (Math|Vector2D|BreakVector2D RT%20Size))
  (bind _pressure_iteration_mid (Variables|Debug|MIDs|GetPressure%20Iteration%20MID))
  (Rendering|Material|SetVectorParameterValue _pressure_iteration_mid "UVScaleAndOffset" UV%20Scale%20and%20Offset :!class /Script/Engine.MaterialInstanceDynamic)
  (Rendering|Material|SetVectorParameterValue _pressure_iteration_mid "XYFrames" (Math|Color|MakeColor _x _y) :!class /Script/Engine.MaterialInstanceDynamic)
  (Rendering|Material|SetVectorParameterValue _pressure_iteration_mid "Resolution" (Math|Color|MakeColor _x_3 _y_3) :!class /Script/Engine.MaterialInstanceDynamic)
  (Rendering|Material|SetScalarParameterValue _pressure_iteration_mid "Divisor" Divisor :!class /Script/Engine.MaterialInstanceDynamic)
  (Rendering|Material|SetVectorParameterValue _pressure_iteration_mid "SampleXYFrames" (Math|Color|MakeColor _x_2 _y_2) :!class /Script/Engine.MaterialInstanceDynamic)
  (for _ (range Iterations)
    (Math|Integer|IncrementInt (Variables|Default|GetPressure%20Index)
      (:%20%20
        (Rendering|Material|SetTextureParameterValue (Variables|Debug|MIDs|GetPressure%20Iteration%20MID) "Divergence" Divergence)
        (Rendering|Material|SetTextureParameterValue (Variables|Debug|MIDs|GetPressure%20Iteration%20MID) "Pressure" Pressure)
        (bind (_canvas _size _context) (Rendering|BeginDrawCanvastoRenderTarget Scratch%20RT))
        (Utilities|IsValid _canvas
          (:Is%20Valid
            (Canvas|DrawMaterial :self _canvas :RenderMaterial (Variables|Debug|MIDs|GetPressure%20Iteration%20MID) :ScreenSize RT%20Size :CoordinateSize "(X=1.000000,Y=1.000000)" :Rotation 0.0 :PivotPoint "(X=0.500000,Y=0.500000)")
            (Rendering|EndDrawCanvastoRenderTarget _context)
            (CallFunction|SwapRenderTargets Pressure Scratch%20RT)))))))

(fn ReleaseRTs
  (Variables|Debug|RenderTargets|SetColor%20Texture "")
  (Variables|Debug|RenderTargets|SetVelocity%20Texture "")
  (Variables|Debug|RenderTargets|SetAdvected%20Velocity%20Texture "")
  (Variables|Debug|RenderTargets|SetPressure%20Texture "")
  (Variables|Debug|RenderTargets|SetPressure%20Scratch%20RT "")
  (Variables|Debug|RenderTargets|SetDivergence%20Texture "")
  (Variables|Debug|RenderTargets|SetTemp%20RT%20Ref "")
  (Variables|Debug|RenderTargets|SetObstacle%20Texture "")
  (Variables|Debug|RenderTargets|SetHalfres%20Pressure "")
  (Variables|Debug|RenderTargets|SetHalfres%20Scratch "")
  (Variables|Debug|RenderTargets|SetQuarterres%20Pressure "")
  (Variables|Debug|RenderTargets|SetQuarterres%20Scratch ""))

(fn ReSampleRT
  (params (XY%20Frames%20to%20Sample :type Vector%202D%20Structure) (XY%20Frames%20to%20Encode :type Vector%202D%20Structure) (UVScale :type Vector%202D%20Structure :default "(X=1.000000,Y=1.000000)") (VolumeResolution :type Vector) (Source%20RT :type Texture%20Render%20Target%202D%20Object%20Reference) (Destination%20RT :type Texture%20Render%20Target%202D%20Object%20Reference) (RT%20Position :type Vector%202D%20Structure) (RT%20Size :type Vector%202D%20Structure))
  (bind (_x _y) (Math|Vector2D|BreakVector2D UVScale))
  (bind (_x_2 _y_2) (Math|Vector2D|BreakVector2D XY%20Frames%20to%20Sample))
  (bind (_x_3 _y_3) (Math|Vector2D|BreakVector2D XY%20Frames%20to%20Encode))
  (bind _resample_rt_mid (Variables|Default|GetReSample%20RT%20MID))
  (Rendering|Material|SetVectorParameterValue (Variables|Default|GetReSample%20RT%20MID) "SampleXYFrames" (Math|Color|MakeColor _x_2 _y_2) :!class /Script/Engine.MaterialInstanceDynamic)
  (Rendering|Material|SetVectorParameterValue _resample_rt_mid "XYFrames" (Math|Color|MakeColor _x_3 _y_3) :!class /Script/Engine.MaterialInstanceDynamic)
  (Rendering|Material|SetVectorParameterValue _resample_rt_mid "UVScale" (Math|Color|MakeColor _x _y) :!class /Script/Engine.MaterialInstanceDynamic)
  (Rendering|Material|SetVectorParameterValue _resample_rt_mid "VolumeResolution" (Math|Conversions|ToLinearColor(Vector) VolumeResolution) :!class /Script/Engine.MaterialInstanceDynamic)
  (Rendering|Material|SetTextureParameterValue (Variables|Default|GetReSample%20RT%20MID) "RT" Source%20RT)
  (Rendering|ClearRenderTarget2D Destination%20RT)
  (bind (_canvas _size _context) (Rendering|BeginDrawCanvastoRenderTarget Destination%20RT))
  (Utilities|IsValid _canvas
    (:Is%20Valid
      (Canvas|DrawMaterial :self _canvas :RenderMaterial (Variables|Default|GetReSample%20RT%20MID) :ScreenPosition RT%20Position :ScreenSize RT%20Size :CoordinateSize "(X=1.000000,Y=1.000000)" :Rotation 0.0 :PivotPoint "(X=0.500000,Y=0.500000)")
      (Rendering|EndDrawCanvastoRenderTarget _context))))

(fn Reset
  (CallFunction|SeedColor+Preview))

(fn SetMIDParameters
  (bind (_x _y) (Math|Vector2D|BreakVector2D (Variables|Debug|GetXY%20Frames)))
  (bind _xy_frames (Variables|Debug|GetXY%20Frames))
  (bind (_x_2 _y_2) (Math|Vector2D|BreakVector2D _xy_frames))
  (bind _returnvalue (Math|Color|MakeColor _x_2 _y_2))
  (bind _p (Variables|SeedColorSetup|GetP))
  (bind _location (.location _p))
  (bind _returnvalue_2 (/ _location 256))
  (bind _returnvalue_3 (- _returnvalue_2 "-0.500000,-0.500000,-0.500000"))
  (bind _returnvalue_4 (Math|Conversions|ToLinearColor(Vector) _returnvalue_3))
  (bind (_x_3 _y_3) (Math|Vector2D|BreakVector2D (Variables|Debug|GetRT%20Size)))
  (bind _color_seed_mid (Variables|Debug|MIDs|GetColor%20Seed%20MID))
  (bind _velocity_seed_mid (Variables|Debug|MIDs|GetVelocity%20Seed%20MID))
  (bind _custom_advection_scalar_params (Variables|Simulation|GetCustom%20Advection%20Scalar%20Params))
  (bind _custom_advection_vector_params (Variables|Simulation|GetCustom%20Advection%20Vector%20Params))
  (bind _velocity_seed_mid_2 (Variables|Debug|MIDs|GetVelocity%20Seed%20MID))
  (bind _forcemids (Variables|Debug|MIDs|GetForceMIDs))
  (bind _velocity_seed_mid_3 (Variables|Debug|MIDs|GetVelocity%20Seed%20MID))
  (bind _advect_mid (Variables|Debug|MIDs|GetAdvect%20MID))
  (bind _velocity_seed_mid_4 (Variables|Debug|MIDs|GetVelocity%20Seed%20MID))
  (bind _advect_mid_2 (Variables|Debug|MIDs|GetAdvect%20MID))
  (bind _custom_forces (Variables|Forces|GetCustom%20Forces))
  (Utilities|FlowControl|Sequence
    (:then_0
      ;[1 1 1] Simulation MIDs
      (Rendering|Material|SetTextureParameterValue (Variables|Debug|MIDs|GetAdvect%20MID) "Obstacle" (select (Utilities|Enum|Equal(Enum) (Variables|Obstacles|GetObstacle%20Mode) "NewEnumerator0") "/Engine/Functions/Engine_MaterialFunctions02/PivotPainter2/Black_1x1_EXR_Texture.Black_1x1_EXR_Texture" (Variables|Debug|RenderTargets|GetObstacle%20Texture)))
      (Rendering|Material|SetVectorParameterValue (Variables|Debug|MIDs|GetRaymarch%20Velocity%20MID) "VolumeResolution" (Math|Conversions|ToLinearColor(Vector) (Variables|Debug|GetVolumeResolution)) :!class /Script/Engine.MaterialInstanceDynamic)
      (Rendering|Material|SetVectorParameterValue (Variables|Debug|MIDs|GetAdvect%20MID) "Resolution" (Math|Color|MakeColor _x_3 _y_3) :!class /Script/Engine.MaterialInstanceDynamic)
      (Rendering|Material|SetVectorParameterValue (Variables|Debug|MIDs|GetAdvect%20MID) "XYFrames" (Math|Color|MakeColor _x _y) :!class /Script/Engine.MaterialInstanceDynamic)
      (Rendering|Material|SetScalarParameterValue _advect_mid "Temp Buoyancy" (Variables|Simulation|GetTemperature%20Buoyancy) :!class /Script/Engine.MaterialInstanceDynamic)
      (Rendering|Material|SetScalarParameterValue _advect_mid "Temperature Dampening" (- 1 (Variables|Simulation|GetTemperature%20Dampening)) :!class /Script/Engine.MaterialInstanceDynamic)
      (Rendering|Material|SetScalarParameterValue _advect_mid "Color Dampening" (- 1 (Variables|Simulation|GetDensity%20Dampening)) :!class /Script/Engine.MaterialInstanceDynamic)
      (Rendering|Material|SetScalarParameterValue _advect_mid "Color Erosion" (Variables|Simulation|GetDensity%20Erosion) :!class /Script/Engine.MaterialInstanceDynamic)
      (Rendering|Material|SetScalarParameterValue _advect_mid "Constant Curl" (Variables|Simulation|GetConstant%20Curl%20Strength) :!class /Script/Engine.MaterialInstanceDynamic)
      (Rendering|Material|SetScalarParameterValue _advect_mid "Constant Curl Tiling" (Variables|Simulation|GetConstant%20Curl%20Tiling) :!class /Script/Engine.MaterialInstanceDynamic)
      ;[1 1 1] Initial Seed Location Setup
      (Rendering|Material|SetVectorParameterValue (Variables|Debug|MIDs|GetColor%20Seed%20MID) "P" _returnvalue_4 :!class /Script/Engine.MaterialInstanceDynamic)
      (Rendering|Material|SetVectorParameterValue (Variables|Debug|MIDs|GetAdvect%20MID) "P2" _returnvalue_4 :!class /Script/Engine.MaterialInstanceDynamic)
      (Rendering|Material|SetScalarParameterValue _color_seed_mid "SourceRadius" (Variables|SeedColorSetup|GetSource%20Radius) :!class /Script/Engine.MaterialInstanceDynamic)
      (Rendering|Material|SetScalarParameterValue _color_seed_mid "Hardness" (Variables|SeedColorSetup|GetEdge%20Hardness) :!class /Script/Engine.MaterialInstanceDynamic)
      (Rendering|Material|SetScalarParameterValue _color_seed_mid "Noise Strength" (Variables|SeedColorSetup|GetNoise%20Strength) :!class /Script/Engine.MaterialInstanceDynamic)
      (Rendering|Material|SetScalarParameterValue _color_seed_mid "Density Multiplier" (Variables|SeedColorSetup|GetDensity%20Multiplier) :!class /Script/Engine.MaterialInstanceDynamic)
      ;[1 1 1] Scale
      (Rendering|Material|SetScalarParameterValue (Variables|Debug|MIDs|GetAdvect%20MID) "ScalePerFrame" (Variables|Default|GetScale%20Per%20Frame) :!class /Script/Engine.MaterialInstanceDynamic)
      ;[1 1 1] Velocity Field Seeding
      (Rendering|Material|SetScalarParameterValue _velocity_seed_mid_4 "Temp" (Variables|SeedColorSetup|GetSeed%20Temperature%20Multiplier) :!class /Script/Engine.MaterialInstanceDynamic)
      (Rendering|Material|SetScalarParameterValue _velocity_seed_mid_4 "Force" (Variables|SeedVelocitySetup|GetOverall%20Velocity%20Multiplier) :!class /Script/Engine.MaterialInstanceDynamic)
      (Rendering|Material|SetScalarParameterValue _velocity_seed_mid_4 "Radial Force" (Variables|SeedVelocitySetup|GetRadial%20Force) :!class /Script/Engine.MaterialInstanceDynamic)
      (Rendering|Material|SetVectorParameterValue _velocity_seed_mid_2 "Offset" (Math|Conversions|ToLinearColor(Vector) (Variables|SeedVelocitySetup|GetNoise%20Field%20Pos%20Offset)) :!class /Script/Engine.MaterialInstanceDynamic)
      (Rendering|Material|SetScalarParameterValue _velocity_seed_mid_2 "Temperature Multiplier" (Variables|SeedColorSetup|GetSeed%20Temperature%20Multiplier) :!class /Script/Engine.MaterialInstanceDynamic)
      (Rendering|Material|SetScalarParameterValue _velocity_seed_mid_2 "Heat Inset" (Variables|SeedColorSetup|GetTemperature%20Inset) :!class /Script/Engine.MaterialInstanceDynamic)
      (Rendering|Material|SetScalarParameterValue _velocity_seed_mid_3 "Curl 1 Tiling" (Variables|SeedVelocitySetup|GetCurl%201%20Tiling) :!class /Script/Engine.MaterialInstanceDynamic)
      (Rendering|Material|SetScalarParameterValue _velocity_seed_mid_3 "Curl 1 Strength" (Variables|SeedVelocitySetup|GetCurl%201%20Strength) :!class /Script/Engine.MaterialInstanceDynamic)
      (Rendering|Material|SetScalarParameterValue _velocity_seed_mid "Curl 2 Tiling" (Variables|SeedVelocitySetup|GetCurl%202%20Tiling) :!class /Script/Engine.MaterialInstanceDynamic)
      (Rendering|Material|SetScalarParameterValue _velocity_seed_mid "Curl 2 Strength" (Variables|SeedVelocitySetup|GetCurl%202%20Strength) :!class /Script/Engine.MaterialInstanceDynamic)
      (Rendering|Material|SetScalarParameterValue (Variables|Debug|MIDs|GetRaymarch%20Velocity%20MID) "Density Mask" (Variables|Debug|GetVelocity%20Density%20Mask%20Amount) :!class /Script/Engine.MaterialInstanceDynamic)
      (Rendering|Material|SetScalarParameterValue (Variables|Debug|MIDs|GetRaymarch%20Color%20and%20Opacity%20MID) "PressurePreview" (Variables|Debug|GetDebug%20Pressure%20Strength) :!class /Script/Engine.MaterialInstanceDynamic))
    (:then_1
      ;[1 1 1] Custom Advection Params
      (bind _keys (Utilities|Map|Keys _custom_advection_scalar_params))
      (for _array_element _keys
        (Rendering|Material|SetScalarParameterValue _advect_mid_2 _array_element (Utilities|Map|Find _custom_advection_scalar_params _array_element) :!class /Script/Engine.MaterialInstanceDynamic)
        (Rendering|Material|SetScalarParameterValue _advect_mid_2 _array_element (Utilities|Map|Find _custom_advection_scalar_params _array_element) :!class /Script/Engine.MaterialInstanceDynamic))
      (bind _keys_2 (Utilities|Map|Keys _custom_advection_vector_params))
      (for _array_element_2 _keys_2
        (Rendering|Material|SetVectorParameterValue _advect_mid_2 _array_element_2 (Utilities|Map|Find _custom_advection_vector_params _array_element_2) :!class /Script/Engine.MaterialInstanceDynamic)
        (Rendering|Material|SetVectorParameterValue _advect_mid_2 _array_element_2 (Utilities|Map|Find _custom_advection_vector_params _array_element_2) :!class /Script/Engine.MaterialInstanceDynamic)))
    (:then_2
      ;[1 1 1] Custom Force MIDs
      (bind _keys_3 (Utilities|Map|Keys (Variables|Forces|GetCustom%20Forces)))
      (for _array_element_3 _keys_3
        (Utilities|IsValid (Utilities|Struct|BreakFluidSimForceMIDs (Utilities|Map|Find _forcemids _array_element_3))
          (:Is%20Valid
            (Rendering|Material|SetVectorParameterValue (Utilities|Struct|BreakFluidSimForceMIDs (Utilities|Map|Find _forcemids _array_element_3)) "XYFrames" _returnvalue :!class /Script/Engine.MaterialInstanceDynamic))
          (:Is%20Not%20Valid
            _))
        (Utilities|IsValid (.ColorMID_4_C7CC5CFA43627B4C5D0617AEDEE80F8F (Utilities|Struct|BreakFluidSimForceMIDs (Utilities|Map|Find _forcemids _array_element_3)))
          (:Is%20Valid
            (Rendering|Material|SetVectorParameterValue (.ColorMID_4_C7CC5CFA43627B4C5D0617AEDEE80F8F (Utilities|Struct|BreakFluidSimForceMIDs (Utilities|Map|Find _forcemids _array_element_3))) "XYFrames" _returnvalue :!class /Script/Engine.MaterialInstanceDynamic))
          (:Is%20Not%20Valid
            _))
        (bind _keys_4 (Utilities|Map|Keys (.MaterialScalarParameters_8_8B0BF3884A694440B4614CBAF75FC65F (Utilities|Struct|BreakFluidSimCustomForce (Utilities|Map|Find _custom_forces _array_element_3)))))
        (for _array_element_4 _keys_4
          (Utilities|IsValid (Utilities|Struct|BreakFluidSimForceMIDs (Utilities|Map|Find _forcemids _array_element_3))
            (:Is%20Valid
              (Rendering|Material|SetScalarParameterValue (Utilities|Struct|BreakFluidSimForceMIDs (Utilities|Map|Find _forcemids _array_element_3)) _array_element_4 (Utilities|Map|Find (.MaterialScalarParameters_8_8B0BF3884A694440B4614CBAF75FC65F (Utilities|Struct|BreakFluidSimCustomForce (Utilities|Map|Find _custom_forces _array_element_3))) _array_element_4) :!class /Script/Engine.MaterialInstanceDynamic))
            (:Is%20Not%20Valid
              _))
          (Utilities|IsValid (.ColorMID_4_C7CC5CFA43627B4C5D0617AEDEE80F8F (Utilities|Struct|BreakFluidSimForceMIDs (Utilities|Map|Find _forcemids _array_element_3)))
            (:Is%20Valid
              (Rendering|Material|SetScalarParameterValue (.ColorMID_4_C7CC5CFA43627B4C5D0617AEDEE80F8F (Utilities|Struct|BreakFluidSimForceMIDs (Utilities|Map|Find _forcemids _array_element_3))) _array_element_4 (Utilities|Map|Find (.MaterialScalarParameters_8_8B0BF3884A694440B4614CBAF75FC65F (Utilities|Struct|BreakFluidSimCustomForce (Utilities|Map|Find _custom_forces _array_element_3))) _array_element_4) :!class /Script/Engine.MaterialInstanceDynamic))))
        (bind _keys_5 (Utilities|Map|Keys (.MaterialVectorParameters_11_57DB292F489BB4FB3E7BE2A66B0245D4 (Utilities|Struct|BreakFluidSimCustomForce (Utilities|Map|Find _custom_forces _array_element_3)))))
        (for _array_element_5 _keys_5
          (Utilities|IsValid (Utilities|Struct|BreakFluidSimForceMIDs (Utilities|Map|Find _forcemids _array_element_3))
            (:Is%20Valid
              (Rendering|Material|SetVectorParameterValue (Utilities|Struct|BreakFluidSimForceMIDs (Utilities|Map|Find _forcemids _array_element_3)) _array_element_5 (Utilities|Map|Find (.MaterialVectorParameters_11_57DB292F489BB4FB3E7BE2A66B0245D4 (Utilities|Struct|BreakFluidSimCustomForce (Utilities|Map|Find _custom_forces _array_element_3))) _array_element_5) :!class /Script/Engine.MaterialInstanceDynamic))
            (:Is%20Not%20Valid
              _))
          (Utilities|IsValid (.ColorMID_4_C7CC5CFA43627B4C5D0617AEDEE80F8F (Utilities|Struct|BreakFluidSimForceMIDs (Utilities|Map|Find _forcemids _array_element_3)))
            (:Is%20Valid
              (Rendering|Material|SetVectorParameterValue (.ColorMID_4_C7CC5CFA43627B4C5D0617AEDEE80F8F (Utilities|Struct|BreakFluidSimForceMIDs (Utilities|Map|Find _forcemids _array_element_3))) _array_element_5 (Utilities|Map|Find (.MaterialVectorParameters_11_57DB292F489BB4FB3E7BE2A66B0245D4 (Utilities|Struct|BreakFluidSimCustomForce (Utilities|Map|Find _custom_forces _array_element_3))) _array_element_5) :!class /Script/Engine.MaterialInstanceDynamic))))))))

(fn SetResolution
  (bind (_x _y) (Math|Vector2D|BreakVector2D (Variables|Debug|GetRT%20Size)))
  (bind _volume_resolution (Variables|BasicSetup|GetVolume%20Resolution))
  (Variables|Debug|SetNum%20Pressure%20Divisions (Utilities|Select :Index _volume_resolution :NewEnumerator0 3 :NewEnumerator1 3 :NewEnumerator2 2 :NewEnumerator3 3 :NewEnumerator4 2 :NewEnumerator5 3 :NewEnumerator6 2 :NewEnumerator7 3))
  (Variables|BasicSetup|SetPressure%20Divisor (Math|Integer|Max(Integer) (Variables|BasicSetup|GetPressure%20Divisor) 1))
  (Variables|Debug|SetXY%20Frames (Utilities|Select :Index _volume_resolution :NewEnumerator0 "(X=32.000000,Y=1.000000)" :NewEnumerator1 "(X=8.000000,Y=8.000000)" :NewEnumerator2 "(X=10.000000,Y=10.000000)" :NewEnumerator3 "(X=16.000000,Y=8.000000)" :NewEnumerator4 "(X=14.000000,Y=14.000000)" :NewEnumerator5 "(X=16.000000,Y=16.000000)" :NewEnumerator6 "(X=18.000000,Y=18.000000)" :NewEnumerator7 "(X=20.000000,Y=20.000000)"))
  (Variables|Debug|SetRT%20Size (select (Variables|BasicSetup|Get3D) (Utilities|Select :Index _volume_resolution :NewEnumerator0 "(X=1024.000000,Y=32.000000)" :NewEnumerator1 "(X=512.000000,Y=512.000000)" :NewEnumerator2 "(X=1000.000000,Y=1000.000000)" :NewEnumerator3 "(X=2048.000000,Y=1024.000000)" :NewEnumerator4 "(X=2744.000000,Y=2744.000000)" :NewEnumerator5 "(X=4096.000000,Y=4096.000000)" :NewEnumerator6 "(X=5832.000000,Y=5832.000000)" :NewEnumerator7 "(X=8000.000000,Y=8000.000000)") (Math|Vector2D|MakeVector2D (Math|Conversions|ToFloat(Integer) (Variables|BasicSetup|Get2D%20Resolution)) (Math|Conversions|ToFloat(Integer) (Variables|BasicSetup|Get2D%20Resolution)))))
  (Variables|Debug|SetVolumeResolution (Math|Conversions|ToVector(Float) (Utilities|Select :Index (Variables|BasicSetup|GetVolume%20Resolution) :NewEnumerator0 32 :NewEnumerator1 64 :NewEnumerator2 100 :NewEnumerator3 128 :NewEnumerator4 196 :NewEnumerator5 256 :NewEnumerator6 324 :NewEnumerator7 400)))
  (Variables|Debug|SetHalfRes%20XY%20Frames (Utilities|Select :Index (Variables|BasicSetup|GetVolume%20Resolution) :NewEnumerator0 "(X=16.000000,Y=1.000000)" :NewEnumerator1 "(X=8.000000,Y=4.000000)" :NewEnumerator2 "(X=10.000000,Y=5.000000)" :NewEnumerator3 "(X=16.000000,Y=4.000000)" :NewEnumerator4 "(X=14.000000,Y=7.000000)" :NewEnumerator5 "(X=16.000000,Y=8.000000)" :NewEnumerator6 "(X=18.000000,Y=9.000000)" :NewEnumerator7 "(X=20.000000,Y=10.000000)"))
  (Variables|Debug|SetHalfres%20RT%20Size (Utilities|Select :Index (Variables|BasicSetup|GetVolume%20Resolution) :NewEnumerator0 "(X=256.000000,Y=16.000000)" :NewEnumerator1 "(X=256.000000,Y=128.000000)" :NewEnumerator2 "(X=500.000000,Y=250.000000)" :NewEnumerator3 "(X=1024.000000,Y=256.000000)" :NewEnumerator4 "(X=1372.000000,Y=686.000000)" :NewEnumerator5 "(X=2048.000000,Y=1024.000000)" :NewEnumerator6 "(X=2916.000000,Y=1458.000000)" :NewEnumerator7 "(X=4000.000000,Y=2000.000000)"))
  (Variables|Debug|SetQuarterRes%20XY%20Frames (Utilities|Select :Index (Variables|BasicSetup|GetVolume%20Resolution) :NewEnumerator0 "(X=8.000000,Y=1.000000)" :NewEnumerator1 "(X=8.000000,Y=2.000000)" :NewEnumerator2 "(X=5.000000,Y=5.000000)" :NewEnumerator3 "(X=16.000000,Y=2.000000)" :NewEnumerator4 "(X=7.000000,Y=7.000000)" :NewEnumerator5 "(X=8.000000,Y=8.000000)" :NewEnumerator6 "(X=9.000000,Y=9.000000)" :NewEnumerator7 "(X=10.000000,Y=10.000000)"))
  (Variables|Debug|SetQuarterRes%20RT%20Size (Utilities|Select :Index (Variables|BasicSetup|GetVolume%20Resolution) :NewEnumerator0 "(X=64.000000,Y=8.000000)" :NewEnumerator1 "(X=128.000000,Y=32.000000)" :NewEnumerator2 "(X=125.000000,Y=125.000000)" :NewEnumerator3 "(X=512.000000,Y=64.000000)" :NewEnumerator4 "(X=343.000000,Y=343.000000)" :NewEnumerator5 "(X=512.000000,Y=512.000000)" :NewEnumerator6 "(X=729.000000,Y=729.000000)" :NewEnumerator7 "(X=1000.000000,Y=1000.000000)"))
  (Variables|Debug|SetEighthRes%20XY%20Frames (Utilities|Select :Index (Variables|BasicSetup|GetVolume%20Resolution) :NewEnumerator0 "(X=4.000000,Y=1.000000)" :NewEnumerator1 "(X=8.000000,Y=1.000000)" :NewEnumerator3 "(X=16.000000,Y=1.000000)" :NewEnumerator4 "(X=0.000000,Y=0.000000)" :NewEnumerator5 "(X=8.000000,Y=4.000000)" :NewEnumerator7 "(X=10.000000,Y=5.000000)"))
  (Variables|Debug|SetEighthRes%20RT%20Size (Utilities|Select :Index (Variables|BasicSetup|GetVolume%20Resolution) :NewEnumerator0 "(X=16.000000,Y=4.000000)" :NewEnumerator1 "(X=64.000000,Y=8.000000)" :NewEnumerator3 "(X=256.000000,Y=16.000000)" :NewEnumerator5 "(X=256.000000,Y=128.000000)" :NewEnumerator7 "(X=500.000000,Y=250.000000)"))
  (Variables|Debug|SetPressure%20RT%20Size (Math|Vector2D|MakeVector2D (Math|Conversions|ToFloat(Integer) (/ (Math|Float|Truncate _x) (Variables|BasicSetup|GetPressure%20Divisor))) (Math|Conversions|ToFloat(Integer) (/ (Math|Float|Truncate _y) (Variables|BasicSetup|GetPressure%20Divisor))))))

(fn SetVolumeDisplayMode
  (params (Index :type VolumeDisplayMaterial%20Enum :default "NewEnumerator0"))
  (Rendering|Material|SetMaterial (Variables|Default|GetS_Box_InsideOut_01) 0 (Utilities|Select :Index Index :NewEnumerator0 (Variables|Debug|MIDs|GetRaymarch%20Color%20and%20Opacity%20MID) :NewEnumerator1 (Variables|Debug|MIDs|GetRaymarch%20Velocity%20MID) :NewEnumerator2 (Variables|Debug|MIDs|GetRaymarch%20Normals%20MID) :NewEnumerator3 (Variables|Debug|MIDs|GetRaymarch%20SH%20MID) :NewEnumerator4 (Variables|Debug|MIDs|GetRaymarch%20Temperature%20MID))))

(fn ShowVelocity
  (Utilities|FlowControl|FlipFlop
    (:A
      (Rendering|Material|SetMaterial (Variables|Default|GetS_Box_InsideOut_01) 0 (Variables|Debug|MIDs|GetRaymarch%20Color%20and%20Opacity%20MID)))
    (:B
      (Rendering|Material|SetMaterial (Variables|Default|GetS_Box_InsideOut_01) 0 (Variables|Debug|MIDs|GetRaymarch%20Velocity%20MID)))))

(fn StartSimulation)

(fn Stop)

(fn SwapRenderTargets
  (params (RT%20A :type Texture%20Render%20Target%202D%20Object%20Reference%20%28by%20ref%29) (RT%20B :type Texture%20Render%20Target%202D%20Object%20Reference%20%28by%20ref%29))
  (Variables|Debug|RenderTargets|SetTemp%20RT%20Ref RT%20A)
  (Variables|SetBy-RefVar RT%20A RT%20B)
  (Variables|SetBy-RefVar RT%20B (Variables|Debug|RenderTargets|GetTemp%20RT%20Ref)))

(fn UpdateDynamicMIDParams
  ;[1 1 1] Scale
  (Rendering|Material|SetScalarParameterValue (Variables|Debug|MIDs|GetAdvect%20MID) "CurrentScale" (Variables|Debug|GetCurScale) :!class /Script/Engine.MaterialInstanceDynamic))

(fn ConstructionScript
  (Utilities|ConstructionSetup)
  (if (not (Variables|Debug|GetSimulating))
    (CallFunction|SeedColor+Preview)
    (if (not (Variables|BasicSetup|Get3D))
      (CallFunction|UpdateObstacles))))
